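# Material function for the SilentNovel Vane character body that adds an animated "FoulTame" lip wobble to World Position Offset, behind a static switch.
# A sine wave over weighted pre-skinned local position blends negative and positive offsets along the vertex normal.
# A soft-edged UV rectangle masks the displacement to the lip region.
# The wave's time comes from a scalar parameter; the posterized Time and speed chain is left unconnected.
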MaterialFunction /BRCosmetics/Characters/Player/Male/Medium/Bodies/M_MED_SilentNovel_Vane/Materials/MF_WildEstate_FoulTame_WPO
graph {nodes [
  n1 = FunctionInput() {inputType=FunctionInput_MaterialAttributes}
  n2 = GetMaterialAttributes(None=@n1) {attributeGetTypes [#1]}
  n3 = Add(A=@n2.WorldPositionOffset, B=NamedRerouteUsage())
  FunctionOutput(None=StaticSwitchParameter(
    True=SetMaterialAttributes!(MaterialAttributes=@n2.MaterialAttributes, WorldPositionOffset=@n3) {attributeSetTypes [#1]}
    False=@n1
   ) {group=FoulTameWPO, parameterName=UseFoulTameWPO, sortPriority=0})
  n4 = ScalarParameter() {group=FoulTameWPO, parameterName=FoulTame_AnimControl Posterize Time}
  n5 = Time()
  n6 = Max(A=@n4) {constB=0.001}
  n7 = Floor(None=Divide(A=@n5, B=@n6))
  Multiply(
   A=If(A=@n4, "A > B"=Multiply(A=@n7, B=@n6), "A == B"=@n5, "A < B"=@n5)
   B=ScalarParameter() {defaultValue=3, group=FoulTameWPO, parameterName=FoulTame_AnimControl Sine Speed}
  )
  n8 = ScalarParameter() {defaultValue=0.15, group=FoulTameWPO, parameterName=FoulTame Lip UV Min X}
  n9 = ScalarParameter() {defaultValue=0.55, group=FoulTameWPO, parameterName=FoulTame Lip UV Max X}
  MaterialExpressionTextureCoordinate_1 = TextureCoordinate()
  n10 = Convert(None=@MaterialExpressionTextureCoordinate_1) {
   nodeName=Break Float2
   convertInputs [{
     type=Vector2
     expressionInput {
      expression=&self:MaterialExpressionTextureCoordinate_1
      inputName=None
      mask=0
      maskA=0
      maskB=0
      maskG=0
      maskR=0
      outputIndex=0
     }
    }]
   convertOutputs [{type=Scalar}, {type=Scalar}]
  }
  n11 = ScalarParameter() {defaultValue=0.75, group=FoulTameWPO, parameterName=FoulTame Lip UV Min Y}
  n12 = ScalarParameter() {defaultValue=0.925, group=FoulTameWPO, parameterName=FoulTame Lip UV Max Y}
  n13 = Saturate(None=AppendVector(A=InverseLinearInterpolate(A=@n8, B=@n9, Value=@n10), B=InverseLinearInterpolate(A=@n11, B=@n12, Value=@n10)))
  MaterialExpressionOneMinus_1 = OneMinus(None=Abs(None=ConstantBiasScale(None=@n13) {bias=-0.5, scale=2}))
  n14 = Convert(None=@MaterialExpressionOneMinus_1) {
   nodeName=Break Float2
   convertInputs [{
     type=Vector2
     expressionInput {
      expression=&self:MaterialExpressionOneMinus_1
      inputName=None
      mask=0
      maskA=0
      maskB=0
      maskG=0
      maskR=0
      outputIndex=0
     }
    }]
   convertOutputs [{type=Scalar}, {type=Scalar}]
  }
  n15 = InverseLinearInterpolate(B=ScalarParameter() {defaultValue=0.5, group=FoulTameWPO, parameterName=FoulTame Lip UV Padding X}, Value=@n14)
  n16 = InverseLinearInterpolate(B=ScalarParameter() {defaultValue=0.4, group=FoulTameWPO, parameterName=FoulTame Lip UV Padding Y}, Value=@n14)
  NamedRerouteDeclaration(None=Multiply(A=Saturate(None=@n15), B=Saturate(None=@n16))) {name=UV Masking, nodeColor=<rgba 0 1 0.4844 1>}
  NamedRerouteDeclaration(None=ScalarParameter() {group=FoulTameWPO, parameterName=FoulTame_AnimControl Time}) {name=Sine Time, nodeColor=<rgba 1 0.4688 0 1>}
  n17 = VertexNormalWS()
  n18 = Multiply(A=Saturate(None=ScalarParameter() {defaultValue=0.2, group=FoulTameWPO, parameterName="FoulTame_AnimControl WPO Negative Offset (0-1)", sliderMax=1})) {constB=-1}
  n19 = Saturate(None=ScalarParameter() {defaultValue=1, group=FoulTameWPO, parameterName="FoulTame_AnimControl WPO Positive Offset (0-1)", sliderMax=1})
  n20 = NamedRerouteUsage()
  n21 = Multiply(
   A=Normalize(VectorInput=Transform(None=@n17) {transformSourceType=TRANSFORMSOURCE_World, transformType=TRANSFORM_Local})
   B=Multiply(
    A=LinearInterpolate(A=@n18, B=@n19, Alpha=@n20)
    B=ScalarParameter() {defaultValue=0.5, group=FoulTameWPO, parameterName=FoulTame_AnimControl WPO Local Strength}
   )
  )
  NamedRerouteDeclaration(None=Multiply(A=Transform(None=@n21) {transformSourceType=TRANSFORMSOURCE_Local}, B=NamedRerouteUsage())) {name=WPO Out, nodeColor=<rgba 0.1484 0 1 1>}
  MaterialExpressionLocalPosition_1 = LocalPosition() {includedOffsets=ExcludeOffsets, localOrigin=InstancePreSkinning}
  n22 = Convert(None=@MaterialExpressionLocalPosition_1.XYZ) {
   nodeName=Break Float3
   convertInputs [{
     type=Vector3
     expressionInput {
      expression=&self:MaterialExpressionLocalPosition_1
      inputName=None
      mask=1
      maskA=0
      maskB=1
      maskG=1
      maskR=1
      outputIndex=0
     }
    }]
   convertOutputs [{type=Scalar}, {type=Scalar}, {type=Scalar}]
  }
  n23 = Multiply(
   A=@n22
   B=Clamp(None=ScalarParameter() {group=FoulTameWPO, parameterName="FoulTame Sine Period X Contribution (-1 to 1)", sliderMax=1, sliderMin=-1}) {minDefault=-1}
  )
  n24 = ScalarParameter() {defaultValue=-0.4, group=FoulTameWPO, parameterName="FoulTame Sine Period Y Contribution (-1 to 1)", sliderMax=1, sliderMin=-1}
  n25 = ScalarParameter() {defaultValue=-1, group=FoulTameWPO, parameterName="FoulTame Sine Period Z Contribution (-1 to 1)", sliderMax=1, sliderMin=-1}
  n26 = Add(A=Multiply(A=@n22, B=Clamp(None=@n24) {minDefault=-1}), B=Multiply(A=@n22, B=Clamp(None=@n25) {minDefault=-1}) {constB=-1})
  n27 = Add(
   A=Multiply(A=Add(A=@n23, B=@n26), B=ScalarParameter() {defaultValue=0.3, group=FoulTameWPO, parameterName=FoulTame_AnimControl Sine Period})
   B=NamedRerouteUsage()
  )
  NamedRerouteDeclaration(None=ConstantBiasScale(None=Sine(None=@n27))) {name=Sine Wave, nodeColor=<rgba 0.9453 0 1 1>}
 ]}
