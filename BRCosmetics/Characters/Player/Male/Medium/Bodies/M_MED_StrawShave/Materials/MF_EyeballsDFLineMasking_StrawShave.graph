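# Material function for the StrawShave player body that masks the incoming DFLines value around the eyeballs as the eyelids close.
# Blink and inner-squint poses, shaped by a threshold exponent, grow per-eye lid spheres in pre-skinned local space, mirrored across X.
# Lines fade where these overlap fixed, mirrored EyeRadius spheres, staying intact elsewhere.
# Upper-lid close and left-eye mask reroutes are declared but do not reach the output.
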MaterialFunction /BRCosmetics/Characters/Player/Male/Medium/Bodies/M_MED_StrawShave/Materials/MF_EyeballsDFLineMasking_StrawShave
graph {nodes [
  n1 = VertexInterpolator(VS=LocalPosition() {localOrigin=InstancePreSkinning}.XYZ)
  n2 = VectorParameter() {parameterName=EyeOrigin, defaultValue=<rgba 0 0 0 1>}
  n3 = ScalarParameter() {parameterName=EyeRadius}
  n4 = SphereMask(A=@n1.PS, B=@n2.RGB, Radius=@n3)
  n5 = SphereMask(A=@n1.PS, B=Multiply(A=@n2.RGB, B=Constant3Vector() {constant=<rgba -1 1 1 0>}), Radius=@n3)
  n6 = VertexInterpolator(VS=LocalPosition() {localOrigin=InstancePreSkinning}.XYZ)
  n7 = VectorParameter() {parameterName=Sphere Origin, defaultValue=<rgba 0 0 0 1>}
  n8 = ScalarParameter() {defaultValue=5, parameterName=Radius}
  n9 = ScalarParameter() {group=01_EyeAnimation_Poses, parameterName=L_blink_pose, sliderMax=1, sortPriority=1}
  n10 = ScalarParameter() {group=01_EyeAnimation_Poses, parameterName=L_squint_inner_pose, sliderMax=1, sortPriority=5}
  n11 = ScalarParameter() {defaultValue=1, group=01_EyeAnimation_Parameters, parameterName=AnimationThreshold_Eyes, sliderMax=8, sortPriority=30}
  n12 = Saturate(None=Power(Base=Max(A=@n9, B=@n10), Exp=@n11))
  n13 = SphereMask(A=@n6.PS, B=@n7.RGB, Radius=Multiply(A=@n8, B=NamedRerouteDeclaration(None=@n12) {name=L_lowerLid_close_pose, nodeColor=<rgba 0.125 1 0 1>}))
  n14 = ScalarParameter() {group=01_EyeAnimation_Poses, parameterName=R_blink_pose, sliderMax=1, sortPriority=2}
  n15 = ScalarParameter() {group=01_EyeAnimation_Poses, parameterName=R_squint_inner_pose, sliderMax=1, sortPriority=6}
  n16 = Saturate(None=Power(Base=Max(A=@n14, B=@n15), Exp=@n11))
  n17 = SphereMask(
   A=@n6.PS
   B=Multiply(A=@n7.RGB, B=Constant3Vector() {constant=<rgba -1 1 1 0>})
   Radius=Multiply(A=@n8, B=NamedRerouteDeclaration(None=@n16) {name=R_lowerLid_close_pose, nodeColor=<rgba 0.125 1 0 1>})
  )
  n18 = Add(A=OneMinus(None=Add(A=@n4, B=@n5)), B=Multiply(A=OneMinus(None=@n13), B=OneMinus(None=@n17)))
  FunctionOutput(None=Multiply(A=FunctionInput() {inputName=DFLines, inputType=FunctionInput_Scalar}, B=Saturate(None=@n18)))
  n19 = ScalarParameter() {defaultValue=0.35, group=01_EyeAnimation_Parameters, parameterName=Squint_Eyelid_Influence, sliderMax=1, sortPriority=31}
  n20 = Max(
   A=ScalarParameter() {group=01_EyeAnimation_Poses, parameterName=R_upperLidDown_pose, sliderMax=1, sortPriority=4}
   B=Max(A=@n14, B=Multiply(A=@n15, B=@n19) {constB=0.5})
  )
  NamedRerouteDeclaration(None=Saturate(None=Power(Base=@n20, Exp=@n11))) {name=R_upperLid_close_pose, nodeColor=<rgba 0.125 1 0 1>}
  n21 = Max(
   A=ScalarParameter() {group=01_EyeAnimation_Poses, parameterName=L_upperLidDown_pose, sliderMax=1, sortPriority=3}
   B=Max(A=Multiply(A=@n10, B=@n19) {constB=0.5}, B=@n9)
  )
  NamedRerouteDeclaration(None=Saturate(None=Power(Base=@n21, Exp=@n11))) {name=L_upperLid_close_pose, nodeColor=<rgba 0.125 1 0 1>}
  n22 = LocalPosition() {localOrigin=InstancePreSkinning}
  MaterialExpressionNamedRerouteDeclaration_9 = NamedRerouteDeclaration(None=ShaderStageSwitch(PixelShader=VertexInterpolator(VS=@n22.XYZ).PS, VertexShader=@n22.XYZ)) {name=PreskinnedLocPos, nodeColor=<rgba 0.3125 1 0 1>}
  n23 = Convert(None=@MaterialExpressionNamedRerouteDeclaration_9) {
   nodeName=Break Float3
   convertInputs [{
     type=Vector3
     expressionInput {
      expression=&self:MaterialExpressionNamedRerouteDeclaration_9
      inputName=None
      mask=0
      maskA=0
      maskB=0
      maskG=0
      maskR=0
      outputIndex=0
     }
    }]
   convertOutputs [{type=Scalar}, {type=Scalar}, {type=Scalar}]
  }
  NamedRerouteDeclaration(None=Step(X=@n23) {constX=0}) {name=LeftEyeMask, nodeColor=<rgba 0.5703 1 0 1>}
 ]}
