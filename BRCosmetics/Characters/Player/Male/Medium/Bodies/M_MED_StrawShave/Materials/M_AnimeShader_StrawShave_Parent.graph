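# Parent toon material for the StrawShave character, using the Cloth shading model and the shared 2023 anime shader Pre, Calculations and Post functions.
# Switches select the lighting normal override from vertex-color baked normals, a height-faded spherical/cylindrical projection or a baked capsule normal map.
# An eye animation function drives lit and shaded color, eyeball masking adjusts DF lines, and a borrowed WPO function wraps the result.
# Temporal responsiveness differs for eyes and fades with camera depth.
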Material /BRCosmetics/Characters/Player/Male/Medium/Bodies/M_MED_StrawShave/Materials/M_AnimeShader_StrawShave_Parent
graph {
 properties {
  bAutomaticallySetUsageInEditor=false
  bUseMaterialAttributes=true
  bUsedWithClothing=true
  bUsedWithMeshDeformer=true
  bUsedWithMorphTargets=true
  bUsedWithNiagaraMeshParticles=true
  bUsedWithSkeletalMesh=true
  bUsedWithStaticLighting=true
  pixelDepthOffsetMode=PDOM_Legacy
  shadingModel=MSM_Cloth
  usedShadingModels=Cloth
 }
 nodes [
  n1 = MaterialFunctionCall!(CustomUV) {materialFunction=/Game/Global/Materials/Character_Parent_2023/Anime_Parent_2023/MaterialFunctions/MFC_AnimeShader_2023_Pre}
  n2 = Transform(None=@n1.Normal)
  n3 = TextureSampleParameter2D() {
   group=Toon_Shading
   parameterName=BakedCapsuleTangentNormal
   samplerType=SAMPLERTYPE_Normal
   texture=&FlatNormalBlue
  }
  n4 = VertexColor()
  n5 = LocalPosition() {localOrigin=InstancePreSkinning}
  n6 = SmoothStep(
   Min=ScalarParameter() {defaultValue=140, group=Toon_Shading, parameterName=NormalSphereFadeStart}
   Max=ScalarParameter() {defaultValue=150, group=Toon_Shading, parameterName=NormalSphereFadeEnd}
   Value=VertexInterpolator(VS=@n5.Z).PS
  ) {constMax=60, constMin=45}
  n7 = StaticSwitchParameter(True=Multiply(A=NamedRerouteDeclaration(None=@n4.A) {nodeColor=<rgba 0.7344 1 0 1>}, B=@n6), False=@n6) {group=Toon_Shading, parameterName=UseMultiplyVertAlphaForFade}
  n8 = LinearInterpolate(A=@n2, B=Transform(None=@n3.RGB), Alpha=@n7)
  n9 = Subtract(A=@n5.XYZ, B=VectorParameter() {group=Toon_Shading, parameterName=NormalSphere_Location, defaultValue=<rgba 0 0 160 1>}.RGB)
  n10 = Multiply(A=@n9, B=Constant3Vector() {constant=<rgba 1 1 0 0>})
  n11 = ScalarParameter() {defaultValue=1, group=Toon_Shading, parameterName=NormalSphereCylindrical_Blend, sliderMax=1}
  n12 = Transform(None=Normalize(VectorInput=LinearInterpolate(A=@n9, B=@n10, Alpha=@n11))) {transformSourceType=TRANSFORMSOURCE_Local}
  n13 = LinearInterpolate(
   A=@n2
   B=LinearInterpolate(
    A=@n2
    B=VertexInterpolator(VS=@n12).PS
    Alpha=ScalarParameter() {defaultValue=0.75, group=Toon_Shading, parameterName=NormalSphere_Blend, sliderMax=1}
   ) {constAlpha=0}
   Alpha=@n7
  )
  n14 = Multiply(A=@n4) {constB=2}
  n15 = ScalarParameter() {desc="0.0 = surface normals, 1.0 baked vertex color normal", group=02_Anime Parameters, parameterName=VertexBakedNormal_Blend, sliderMax=1}
  n16 = LinearInterpolate(A=@n2, B=Transform(None=Subtract(A=@n14)) {transformSourceType=TRANSFORMSOURCE_Local}, Alpha=Multiply(A=@n15, B=@n4.A))
  n17 = MaterialFunctionCall!(
   CustomLightVector
   OverrideNormal_WS=StaticSwitchParameter(
    True=StaticSwitchParameter(True=@n8, False=@n13) {group=Toon_Shading, parameterName=UseCapsuleTangentNormal}
    False=Normalize(VectorInput=@n16)
   ) {group=Toon_Shading, parameterName=UseSphereCylindricalNormal}
   STT=@n1.STT
   Normal_Map=@n1.STT
   DFL=@n1.STT
   FX=@n1.STT
  ) {materialFunction=/Game/Global/Materials/Character_Parent_2023/Anime_Parent_2023/MaterialFunctions/MFC_AnimeShader_2023_Calculations}
  n18 = MaterialFunctionCall!(DFLines=@n17.DFL_Mask) {materialFunction=&/MF_EyeballsDFLineMasking_StrawShave}
  n19 = VertexColor()
  n20 = TextureSampleParameter2D() {
   group=01_EyeAnimation_Parameters
   parameterName=EyeMask
   samplerSource=SSM_Wrap_WorldGroupSettings
   samplerType=SAMPLERTYPE_Masks
   sortPriority=-32
   texture=&T_WhiteMask
  }
  n21 = StaticSwitchParameter(True=AppendVector(A=@n19, B=@n19), False=@n20.RGBA) {group=01_EyeAnimation_Parameters, parameterName=UseVertexColorForEyeMask, sortPriority=-33}
  n22 = TextureCoordinate() {coordinateIndex=1}
  n23 = Sign(None=Subtract(A=ComponentMask(None=@n22) {r=true}) {constB=0.0001})
  n24 = MaterialFunctionCall!(
   Color Lit=@n1.Color_Lit
   Color Shaded=@n1.Color_Lit
   Use UV2=StaticBoolParameter() {defaultValue=true, group=01_EyeAnimation_Parameters, parameterName=Use UV2 for Eye Animation, sortPriority=-30}
   Mask=StaticSwitchParameter(
    True=ChannelMaskParameter(None=@n21) {group=01_EyeAnimation_Parameters, parameterName=EyeMaskChannel, sortPriority=-32}.R
    False=Saturate(None=@n23)
   ) {group=01_EyeAnimation_Parameters, parameterName=UseEyeMask, sortPriority=-33}
  ) {materialFunction=/BRCosmetics/Characters/Player/Female/Medium/Bodies/F_MED_FruitFire_Nudge/Materials/MF_AnimeEyeAnimation_Porridge}
  StaticSwitchParameter(
   True=SetMaterialAttributes() {attributeSetTypes [#1]}
   False=MaterialFunctionCall!(In=MaterialFunctionCall!(
     Shading_Gradient=@n17.Shading_Gradient
     Threshold_Gradient=@n17.Shading_Gradient
     Specular_Gradient=@n17.Shading_Gradient
     Sheen_Mask=@n17.Shading_Gradient
     Thickness=@n17.Shading_Gradient
     DFL_Mask=@n18.Result
     FX=@n17.Shading_Gradient
     Metallic_Gradient=@n17.Shading_Gradient
     RimLight_Gradient=@n17.Shading_Gradient
     NormalBuffer=@n17.Shading_Gradient
     --------
     CINE_LightSources=@n17.Shading_Gradient
     CINE_LightColor=@n17.Shading_Gradient
     ---------
     Color_Lit=@n24.Lit
     Color_Shaded=@n24.Lit
     SRM=@n1.SRM
     Emissive=@n1.SRM
     DFL_Color=@n1.SRM
     Specular_Color=@n1.SRM
     RimLight_Color=@n1.SRM
    ) {materialFunction=/Game/Global/Materials/Character_Parent_2023/Anime_Parent_2023/MaterialFunctions/MFC_AnimeShader_2023_Post}.Result) {materialFunction=/BRCosmetics/Characters/Player/Male/Medium/Bodies/M_MED_SilentNovel_Vane/Materials/MF_WildEstate_FoulTame_WPO}.Result
  ) {group=Eyelid DF Lines Masking, parameterName=UseEyelidDFLinesMaskingDebug, sortPriority=-1}
  n25 = ScalarParameter() {defaultValue=0.25, parameterName=NonEyesTRMulti}
  n26 = ScalarParameter() {defaultValue=1, parameterName=EyesTRMulti}
  n27 = ScalarParameter() {defaultValue=500, group=01_EyeAnimation_Parameters, parameterName=TR_FadeLength, sliderMax=2, sortPriority=5}
  n28 = ScalarParameter() {defaultValue=500, group=01_EyeAnimation_Parameters, parameterName=TR_StartDistance, sliderMax=2, sortPriority=5}
  TemporalResponsivenessOutput!(None=LinearInterpolate(
    A=LinearInterpolate(A=@n25, B=@n26, Alpha=@n24.EyeMask)
    Alpha=MaterialFunctionCall(Fade Length=@n27, Fade Offset=@n28) {materialFunction=&CameraDepthFade}.Result
   ) {constB=0})
 ]
}
