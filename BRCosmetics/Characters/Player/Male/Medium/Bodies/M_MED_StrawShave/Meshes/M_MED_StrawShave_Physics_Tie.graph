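# Physics asset for the tie on the M_MED_StrawShave male cosmetic body, apparently for secondary cloth-like motion.
# A five-segment capsule chain hangs from a small tie master body on spine_05, with narrow swing and tighter twist limits between segments.
# Large capsules on the pelvis, spine and thighs approximate the body, alongside two front-jacket capsules constrained to the pelvis.
# All constraints outside the tie segments have free swing and twist.
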PhysicsAsset /BRCosmetics/Characters/Player/Male/Medium/Bodies/M_MED_StrawShave/Meshes/M_MED_StrawShave_Physics_Tie
bodies {
 dyn_jacket_fr_l_MSTR [{
   shapeType=Capsule
   length=12
   radius=3
   shapeName=dyn_jacket_fr_l_MSTR_capsule
   center=<6.6974 2.1816 2.4899>
   rotation=<rot -0.1971 -90.1707 -87.8785>
  }]
 dyn_jacket_fr_r_MSTR [{
   shapeType=Capsule
   length=12
   radius=3
   shapeName=dyn_jacket_fr_r_MSTR_capsule
   center=<6.6974 -2.1816 2.1394>
   rotation=<rot 0.3712 -95.9467 -88.0158>
  }]
 dyn_tie_01 [{shapeType=Capsule, length=6, radius=3, shapeName=dyn_tie_01_capsule, center=<4.805 0 1.7751>, rotation=<rot 90 0 0>}]
 dyn_tie_02 [{shapeType=Capsule, length=6, radius=3, shapeName=dyn_tie_02_capsule, center=<4.805 0 2.2027>, rotation=<rot 90 0 0>}]
 dyn_tie_03 [{shapeType=Capsule, length=6, radius=3, shapeName=dyn_tie_03_capsule, center=<4.805 0 2.0263>, rotation=<rot 90 0.0099 0.0099>}]
 dyn_tie_04 [{shapeType=Capsule, length=6, radius=3, shapeName=dyn_tie_04_capsule, center=<4.805 0 2.0148>, rotation=<rot 90 0 0>}]
 dyn_tie_05 [{shapeType=Capsule, length=6, radius=3, shapeName=dyn_tie_05_capsule, center=<4.8724 0 2.2249>, rotation=<rot 90 0 0>}]
 dyn_tie_MSTR [{shapeType=Capsule, length=0.505, radius=0.505, shapeName=dyn_tie_MSTR_capsule, rotation=<rot 0 -90 0>}]
 pelvis [{shapeType=Capsule, length=15, radius=11.9, center=<-2.0733 -1.3709 0>}]
 spine_01 [{shapeType=Capsule, length=15, radius=12, center=<-2.0733 -1.3709 0>}]
 spine_02 [{shapeType=Capsule, length=15, radius=12, shapeName=spine_02_capsule, center=<0 -0.9108 0>}]
 spine_03 [{shapeType=Capsule, length=15, radius=12, shapeName=spine_03_capsule, center=<0 -0.4578 0>}]
 spine_04 [{shapeType=Capsule, length=15, radius=13, shapeName=spine_04_capsule, center=<0 -0.5262 0>}]
 spine_05 [{shapeType=Capsule, length=20, radius=12, shapeName=spine_05_capsule, center=<0 -1.6609 0>}]
 thigh_l [{shapeType=Capsule, length=40, radius=11.5, center=<-20.5 -1.2351 0.5583>, rotation=<rot 90 0 0>}]
 thigh_r [{shapeType=Capsule, length=40, radius=11.5, center=<20.5 1.2351 -0.8238>, rotation=<rot 90 0 0>}]
}
constraints [
 {
  bone1Name=thigh_r
  bone2Name=pelvis
  swing1LimitDegrees=45
  swing1Motion=Free
  swing2LimitDegrees=45
  swing2Motion=Free
  twistLimitDegrees=45
  twistMotion=Free
 }
 {
  bone1Name=dyn_jacket_fr_l_MSTR
  bone2Name=pelvis
  swing1LimitDegrees=45
  swing1Motion=Free
  swing2LimitDegrees=45
  swing2Motion=Free
  twistLimitDegrees=45
  twistMotion=Free
 }
 {
  bone1Name=dyn_jacket_fr_r_MSTR
  bone2Name=pelvis
  swing1LimitDegrees=45
  swing1Motion=Free
  swing2LimitDegrees=45
  swing2Motion=Free
  twistLimitDegrees=45
  twistMotion=Free
 }
 {
  bone1Name=spine_01
  bone2Name=pelvis
  swing1LimitDegrees=45
  swing1Motion=Free
  swing2LimitDegrees=45
  swing2Motion=Free
  twistLimitDegrees=45
  twistMotion=Free
 }
 {
  bone1Name=spine_02
  bone2Name=spine_01
  swing1LimitDegrees=45
  swing1Motion=Free
  swing2LimitDegrees=45
  swing2Motion=Free
  twistLimitDegrees=45
  twistMotion=Free
 }
 {
  bone1Name=spine_03
  bone2Name=spine_02
  swing1LimitDegrees=45
  swing1Motion=Free
  swing2LimitDegrees=45
  swing2Motion=Free
  twistLimitDegrees=45
  twistMotion=Free
 }
 {
  bone1Name=spine_04
  bone2Name=spine_03
  swing1LimitDegrees=45
  swing1Motion=Free
  swing2LimitDegrees=45
  swing2Motion=Free
  twistLimitDegrees=45
  twistMotion=Free
 }
 {
  bone1Name=spine_05
  bone2Name=spine_04
  swing1LimitDegrees=45
  swing1Motion=Free
  swing2LimitDegrees=45
  swing2Motion=Free
  twistLimitDegrees=45
  twistMotion=Free
 }
 {
  bone1Name=dyn_tie_MSTR
  bone2Name=spine_05
  swing1LimitDegrees=45
  swing1Motion=Free
  swing2LimitDegrees=45
  swing2Motion=Free
  twistLimitDegrees=45
  twistMotion=Free
 }
 {bone1Name=dyn_tie_01, bone2Name=dyn_tie_MSTR, swing1LimitDegrees=15, swing2LimitDegrees=15, twistLimitDegrees=2}
 {bone1Name=dyn_tie_02, bone2Name=dyn_tie_01, swing1LimitDegrees=15, swing2LimitDegrees=15, twistLimitDegrees=2}
 {bone1Name=dyn_tie_03, bone2Name=dyn_tie_02, swing1LimitDegrees=15, swing2LimitDegrees=15, twistLimitDegrees=2}
 {bone1Name=dyn_tie_04, bone2Name=dyn_tie_03, swing1LimitDegrees=15, swing2LimitDegrees=15, twistLimitDegrees=2}
 {bone1Name=dyn_tie_05, bone2Name=dyn_tie_04, swing1LimitDegrees=15, swing2LimitDegrees=15, twistLimitDegrees=2}
 {
  bone1Name=thigh_l
  bone2Name=pelvis
  swing1LimitDegrees=45
  swing1Motion=Free
  swing2LimitDegrees=45
  swing2Motion=Free
  twistLimitDegrees=45
  twistMotion=Free
 }
]
