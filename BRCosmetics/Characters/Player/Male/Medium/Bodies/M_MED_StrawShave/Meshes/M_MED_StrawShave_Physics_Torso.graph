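# Physics asset for the torso of the medium male StrawShave player body, setting up jacket dynamics against the lower body.
# Capsules on the pelvis, lowest spine bone, thighs, calves and feet form the body, joined by angularly free constraints.
# Seven two-segment jacket chains hang from small master bodies on the pelvis, with narrow swing and very tight twist limits.
# Extra angularly free constraints tie each front panel's lower segment to its neighbouring side panel.
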PhysicsAsset /BRCosmetics/Characters/Player/Male/Medium/Bodies/M_MED_StrawShave/Meshes/M_MED_StrawShave_Physics_Torso
bodies {
 calf_l [{shapeType=Capsule, length=35, radius=11, shapeName=calf_l_capsule, center=<-22.0586 -0.1727 1.0302>, rotation=<rot 0 -95.9458 -88.4084>}]
 calf_r [{shapeType=Capsule, length=35, radius=11, shapeName=calf_r_capsule, center=<22.0595 0.1726 -1.0014>, rotation=<rot 0 -95.9848 -88.482>}]
 dyn_jacket_bk_01 [{shapeType=Capsule, length=13, radius=3, center=<6 0 4.8>, rotation=<rot 0 0 -90>}]
 dyn_jacket_bk_02 [{shapeType=Capsule, length=8, radius=3, center=<6 0 4.8>, rotation=<rot 0 0 -90>}]
 dyn_jacket_bk_MSTR [{shapeType=Capsule, length=1, radius=1, center=<1 0 0>}]
 dyn_jacket_bk_l_01 [{shapeType=Capsule, length=13, radius=3, center=<6 0 4.8>, rotation=<rot 0 0 -90>}]
 dyn_jacket_bk_l_02 [{shapeType=Capsule, length=13, radius=3, center=<6 0 4.8>, rotation=<rot 0 0 -90>}]
 dyn_jacket_bk_l_MSTR [{shapeType=Capsule, length=1, radius=1, center=<1 0 0>}]
 dyn_jacket_bk_r_01 [{shapeType=Capsule, length=13, radius=3, center=<6 0 4.8>, rotation=<rot 0 0 -90>}]
 dyn_jacket_bk_r_02 [{shapeType=Capsule, length=13, radius=3, center=<6 0 4.8>, rotation=<rot 0 0 -90>}]
 dyn_jacket_bk_r_MSTR [{shapeType=Capsule, length=1, radius=1, center=<1 0 0>}]
 dyn_jacket_fr_l_01 [{shapeType=Capsule, length=13, radius=5, center=<12 0 6>, rotation=<rot 0 0 -90>}]
 dyn_jacket_fr_l_02 [{shapeType=Capsule, length=13, radius=5, center=<12 0 6>, rotation=<rot 0 0 -90>}]
 dyn_jacket_fr_l_MSTR [{shapeType=Capsule, length=1, radius=1, center=<1 0 0>}]
 dyn_jacket_fr_r_01 [{shapeType=Capsule, length=13, radius=5, center=<12 0 4.8>, rotation=<rot 0 0 -90>}]
 dyn_jacket_fr_r_02 [{shapeType=Capsule, length=13, radius=5.5, center=<12 0 4.8>, rotation=<rot 0.0003 -0.0001 -70>}]
 dyn_jacket_fr_r_MSTR [{shapeType=Capsule, length=1, radius=1, center=<1 0 0>}]
 dyn_jacket_l_01 [{shapeType=Capsule, length=13, radius=3, center=<6 0 4.8>, rotation=<rot 0 0 -90>}]
 dyn_jacket_l_02 [{shapeType=Capsule, length=13, radius=3, center=<6 0 4.8>, rotation=<rot 0 0 -90>}]
 dyn_jacket_l_MSTR [{shapeType=Capsule, length=1, radius=1, center=<1 0 0>}]
 dyn_jacket_r_01 [{shapeType=Capsule, length=13, radius=3, center=<6 0 4.8>, rotation=<rot 0 0 -90>}]
 dyn_jacket_r_02 [{shapeType=Capsule, length=13, radius=3, center=<6 0 4.8>, rotation=<rot 0 0 -90>}]
 dyn_jacket_r_MSTR [{shapeType=Capsule, length=1, radius=1, center=<1 0 0>}]
 foot_l [{shapeType=Capsule, length=15, radius=9, shapeName=foot_l_capsule, center=<-4.4978 6.6731 1>, rotation=<rot 0 0 -89>}]
 foot_r [{shapeType=Capsule, length=15, radius=9, shapeName=foot_r_capsule, center=<4.4978 -6.6731 -1>, rotation=<rot 0 0 -89>}]
 pelvis [{shapeType=Capsule, length=10, radius=14.5, center=<-2.5 -1.3709 0>}]
 spine_01 [{shapeType=Capsule, length=9, radius=15, center=<-4 -1.3709 0>}]
 thigh_l [{shapeType=Capsule, length=26.75, radius=13.5, center=<-13.9111 -1.2351 0.5583>, rotation=<rot 90 0 0>}]
 thigh_r [{shapeType=Capsule, length=26.75, radius=13.5, center=<13.9111 1.2351 -0.8238>, rotation=<rot 90 0 0>}]
}
constraints [
 {
  bone1Name=foot_r
  bone2Name=calf_r
  swing1LimitDegrees=45
  swing1Motion=Free
  swing2LimitDegrees=45
  swing2Motion=Free
  twistLimitDegrees=45
  twistMotion=Free
 }
 {
  bone1Name=dyn_jacket_bk_MSTR
  bone2Name=pelvis
  swing1LimitDegrees=45
  swing1Motion=Free
  swing2LimitDegrees=45
  swing2Motion=Free
  twistLimitDegrees=45
  twistMotion=Free
 }
 {bone1Name=dyn_jacket_bk_01, bone2Name=dyn_jacket_bk_MSTR, swing1LimitDegrees=10, swing2LimitDegrees=30, twistLimitDegrees=1}
 {bone1Name=dyn_jacket_bk_02, bone2Name=dyn_jacket_bk_01, swing1LimitDegrees=10, swing2LimitDegrees=30, twistLimitDegrees=1}
 {
  bone1Name=dyn_jacket_bk_l_MSTR
  bone2Name=pelvis
  swing1LimitDegrees=45
  swing1Motion=Free
  swing2LimitDegrees=45
  swing2Motion=Free
  twistLimitDegrees=45
  twistMotion=Free
 }
 {bone1Name=dyn_jacket_bk_l_01, bone2Name=dyn_jacket_bk_l_MSTR, swing1LimitDegrees=10, swing2LimitDegrees=30, twistLimitDegrees=1}
 {bone1Name=dyn_jacket_bk_l_02, bone2Name=dyn_jacket_bk_l_01, swing1LimitDegrees=10, swing2LimitDegrees=30, twistLimitDegrees=1}
 {
  bone1Name=dyn_jacket_bk_r_MSTR
  bone2Name=pelvis
  swing1LimitDegrees=45
  swing1Motion=Free
  swing2LimitDegrees=45
  swing2Motion=Free
  twistLimitDegrees=45
  twistMotion=Free
 }
 {bone1Name=dyn_jacket_bk_r_01, bone2Name=dyn_jacket_bk_r_MSTR, swing1LimitDegrees=10, swing2LimitDegrees=30, twistLimitDegrees=1}
 {bone1Name=dyn_jacket_bk_r_02, bone2Name=dyn_jacket_bk_r_01, swing1LimitDegrees=10, swing2LimitDegrees=30, twistLimitDegrees=1}
 {
  bone1Name=dyn_jacket_fr_l_MSTR
  bone2Name=pelvis
  swing1LimitDegrees=45
  swing1Motion=Free
  swing2LimitDegrees=45
  swing2Motion=Free
  twistLimitDegrees=45
  twistMotion=Free
 }
 {bone1Name=dyn_jacket_fr_l_01, bone2Name=dyn_jacket_fr_l_MSTR, swing1LimitDegrees=25, swing2LimitDegrees=25, twistLimitDegrees=2}
 {bone1Name=dyn_jacket_fr_l_02, bone2Name=dyn_jacket_fr_l_01, swing1LimitDegrees=10, swing2LimitDegrees=25, twistLimitDegrees=2}
 {
  bone1Name=dyn_jacket_fr_r_MSTR
  bone2Name=pelvis
  swing1LimitDegrees=45
  swing1Motion=Free
  swing2LimitDegrees=45
  swing2Motion=Free
  twistLimitDegrees=45
  twistMotion=Free
 }
 {bone1Name=dyn_jacket_fr_r_01, bone2Name=dyn_jacket_fr_r_MSTR, swing1LimitDegrees=10, swing2LimitDegrees=25, twistLimitDegrees=2}
 {bone1Name=dyn_jacket_fr_r_02, bone2Name=dyn_jacket_fr_r_01, swing1LimitDegrees=10, swing2LimitDegrees=25, twistLimitDegrees=2}
 {
  bone1Name=dyn_jacket_l_MSTR
  bone2Name=pelvis
  swing1LimitDegrees=45
  swing1Motion=Free
  swing2LimitDegrees=45
  swing2Motion=Free
  twistLimitDegrees=45
  twistMotion=Free
 }
 {bone1Name=dyn_jacket_l_01, bone2Name=dyn_jacket_l_MSTR, swing1LimitDegrees=10, swing2LimitDegrees=30, twistLimitDegrees=2}
 {bone1Name=dyn_jacket_l_02, bone2Name=dyn_jacket_l_01, swing1LimitDegrees=10, swing2LimitDegrees=30, twistLimitDegrees=2}
 {
  bone1Name=dyn_jacket_r_MSTR
  bone2Name=pelvis
  swing1LimitDegrees=45
  swing1Motion=Free
  swing2LimitDegrees=45
  swing2Motion=Free
  twistLimitDegrees=45
  twistMotion=Free
 }
 {bone1Name=dyn_jacket_r_01, bone2Name=dyn_jacket_r_MSTR, swing1LimitDegrees=10, swing2LimitDegrees=30, twistLimitDegrees=2}
 {bone1Name=dyn_jacket_r_02, bone2Name=dyn_jacket_r_01, swing1LimitDegrees=10, swing2LimitDegrees=30, twistLimitDegrees=2}
 {
  bone1Name=spine_01
  bone2Name=pelvis
  swing1LimitDegrees=45
  swing1Motion=Free
  swing2LimitDegrees=45
  swing2Motion=Free
  twistLimitDegrees=45
  twistMotion=Free
 }
 {
  bone1Name=thigh_l
  bone2Name=pelvis
  swing1LimitDegrees=45
  swing1Motion=Free
  swing2LimitDegrees=45
  swing2Motion=Free
  twistLimitDegrees=45
  twistMotion=Free
 }
 {
  bone1Name=calf_l
  bone2Name=thigh_l
  swing1LimitDegrees=45
  swing1Motion=Free
  swing2LimitDegrees=45
  swing2Motion=Free
  twistLimitDegrees=45
  twistMotion=Free
 }
 {
  bone1Name=foot_l
  bone2Name=calf_l
  swing1LimitDegrees=45
  swing1Motion=Free
  swing2LimitDegrees=45
  swing2Motion=Free
  twistLimitDegrees=45
  twistMotion=Free
 }
 {
  bone1Name=thigh_r
  bone2Name=pelvis
  swing1LimitDegrees=45
  swing1Motion=Free
  swing2LimitDegrees=45
  swing2Motion=Free
  twistLimitDegrees=45
  twistMotion=Free
 }
 {
  bone1Name=calf_r
  bone2Name=thigh_r
  swing1LimitDegrees=45
  swing1Motion=Free
  swing2LimitDegrees=45
  swing2Motion=Free
  twistLimitDegrees=45
  twistMotion=Free
 }
 {
  bone1Name=dyn_jacket_fr_r_02
  bone2Name=dyn_jacket_r_02
  swing1LimitDegrees=45
  swing1Motion=Free
  swing2LimitDegrees=45
  swing2Motion=Free
  twistLimitDegrees=45
  twistMotion=Free
 }
 {
  bone1Name=dyn_jacket_fr_l_02
  bone2Name=dyn_jacket_l_02
  swing1LimitDegrees=45
  swing1Motion=Free
  swing2LimitDegrees=45
  swing2Motion=Free
  twistLimitDegrees=45
  twistMotion=Free
 }
]
