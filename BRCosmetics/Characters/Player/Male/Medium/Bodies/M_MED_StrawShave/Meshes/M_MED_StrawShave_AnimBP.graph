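# Post-process-style Anim Blueprint for the StrawShave male body cosmetic, animating its tie and jacket on the copied parent pose.
# A Control Rig reads crouch and emote flags, then feeds either chained torso and tie RigidBody sims or weighted tie constraints, chosen by bRigidBodyEnabled.
# Fixed additive bone offsets then reshape the jacket.
# Each update derives state-based gravity for the sims, plus RBAN settings that are stored but unwired in the graph.
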AnimBlueprint /BRCosmetics/Characters/Player/Male/Medium/Bodies/M_MED_StrawShave/Meshes/M_MED_StrawShave_AnimBP
target_skeleton=&/M_MED_StrawShave_Skeleton
graphs [
 AnimGraph {nodes [
   n1 = "Constraint - dyn_tie_01 (PPPPPPPP)"!(
    ComponentPose=Local To Component(LocalPose=Use cached pose 'pre_sim_cache'())
    Alpha=1
    ConstraintWeights_0=0.5865
    ConstraintWeights_1:float=0.3466
    ConstraintWeights_2:float=0.0158
    ConstraintWeights_3:float=0.0148
    ConstraintWeights_4:float=0.0111
    ConstraintWeights_5:float=0.0097
    ConstraintWeights_6:float=0.0083
    ConstraintWeights_7:float=0.0074
   ) -> (Pose) {
    boneToModify {boneName=dyn_tie_01}
    constraintSetup [
     {targetBone {boneName=dyn_tie_01}}
     {targetBone {boneName=spine_05}}
     {targetBone {boneName=spine_01}}
     {targetBone {boneName=clavicle_l}}
     {targetBone {boneName=spine_02}}
     {targetBone {boneName=pelvis}}
     {targetBone {boneName=clavicle_r}}
     {targetBone {boneName=spine_04}}
    ]
    constraintWeights [1,1,1,1,1,1,1,1]
   }
   n2 = "Constraint - dyn_tie_02 (PPPPPPPP)"!(
    ComponentPose=@n1
    Alpha=1
    ConstraintWeights_0=0.5675
    ConstraintWeights_1:float=0.316
    ConstraintWeights_2:float=0.0392
    ConstraintWeights_3:float=0.0323
    ConstraintWeights_4:float=0.0248
    ConstraintWeights_5:float=0.0164
    ConstraintWeights_6:float=0.0019
    ConstraintWeights_7:float=0.0019
   ) -> (Pose) {
    boneToModify {boneName=dyn_tie_02}
    constraintSetup [
     {targetBone {boneName=spine_05}}
     {targetBone {boneName=spine_04}}
     {targetBone {boneName=spine_03}}
     {targetBone {boneName=spine_01}}
     {targetBone {boneName=spine_02}}
     {targetBone {boneName=pelvis}}
     {targetBone {boneName=deform_pec_l}}
     {targetBone {boneName=deform_pec_r}}
    ]
    constraintWeights [1,1,1,1,1,1,1,1]
   }
   n3 = "Constraint - dyn_tie_03 (PPPPPPPP)"!(
    ComponentPose=@n2
    Alpha=1
    ConstraintWeights_0=0.6383
    ConstraintWeights_1:float=0.1565
    ConstraintWeights_2:float=0.1255
    ConstraintWeights_3:float=0.0462
    ConstraintWeights_4:float=0.0258
    ConstraintWeights_5:float=0.0032
    ConstraintWeights_6:float=0.0032
    ConstraintWeights_7:float=0.0014
   ) -> (Pose) {
    boneToModify {boneName=dyn_tie_03}
    constraintSetup [
     {targetBone {boneName=spine_03}}
     {targetBone {boneName=spine_02}}
     {targetBone {boneName=spine_04}}
     {targetBone {boneName=spine_01}}
     {targetBone {boneName=pelvis}}
     {targetBone {boneName=clavicle_l}}
     {targetBone {boneName=clavicle_r}}
     {targetBone {boneName=spine_05}}
    ]
    constraintWeights [1,1,1,1,1,1,1,1]
   }
   n4 = "Constraint - dyn_tie_04 (PPPPPP)"!(
    ComponentPose=@n3
    Alpha=1
    ConstraintWeights_0=0.4166
    ConstraintWeights_1:float=0.2156
    ConstraintWeights_2:float=0.2029
    ConstraintWeights_3:float=0.1613
    ConstraintWeights_4:float=0.0019
    ConstraintWeights_5:float=0.0017
   ) -> (Pose) {
    boneToModify {boneName=dyn_tie_04}
    constraintSetup [
     {targetBone {boneName=spine_02}}
     {targetBone {boneName=spine_01}}
     {targetBone {boneName=pelvis}}
     {targetBone {boneName=spine_03}}
     {targetBone {boneName=deform_groin_r}}
     {targetBone {boneName=thigh_twist_01_r}}
    ]
    constraintWeights [1,1,1,1,1,1]
   }
   n5 = "Constraint - dyn_tie_05 (PPPPPP)"!(
    ComponentPose=@n4
    Alpha=1
    ConstraintWeights_0=0.5502
    ConstraintWeights_1:float=0.2926
    ConstraintWeights_2:float=0.1456
    ConstraintWeights_3:float=0.0051
    ConstraintWeights_4:float=0.0045
    ConstraintWeights_5:float=0.002
   ) -> (Pose) {
    boneToModify {boneName=dyn_tie_05}
    constraintSetup [
     {targetBone {boneName=pelvis}}
     {targetBone {boneName=spine_01}}
     {targetBone {boneName=spine_02}}
     {targetBone {boneName=deform_groin_r}}
     {targetBone {boneName=thigh_twist_01_r}}
     {targetBone {boneName=spine_03}}
    ]
    constraintWeights [1,1,1,1,1,1]
   }
   n6 = Component To Local(ComponentPose=@n5)
   n7 = RigidBody!(
    ComponentPose=Local To Component(LocalPose=Use cached pose 'pre_sim_cache'())
    Alpha
    OverrideWorldGravity:Vector=Get Gravity Override()
    ExternalForce
    SimSpaceSettings:Sim Space Settings Structure=(WorldAlpha=0,VelocityScaleZ=1,DampingAlpha=1,MaxLinearVelocity=10000,MaxAngularVelocity=10000,MaxLinearAcceleration=10000,MaxAngularAcceleration=10000,ExternalLinearDragV=(X=0,Y=0,Z=0),ExternalLinearVelocity=(X=0,Y=0,Z=0),ExternalAngularVelocity=(X=0,Y=0,Z=0))
   ) -> (Pose) {
    bClampLinearTranslationLimitToRefPose=true
    bOverrideWorldGravity=true
    overridePhysicsAsset=&/M_MED_StrawShave_Physics_Torso
    simulationSpace=BaseBoneSpace
    alphaBoolBlend {blendOption=Linear}
    baseBoneRef {boneName=pelvis}
    componentAppliedLinearAccClamp=<0 0 0>
    overrideWorldGravity {z=-980}
   }
   n8 = RigidBody(
    ComponentPose=@n7
    OverrideWorldGravity=Get Gravity Override()
    SimSpaceSettings=(WorldAlpha=0,VelocityScaleZ=1,DampingAlpha=1,MaxLinearVelocity=10000,MaxAngularVelocity=10000,MaxLinearAcceleration=10000,MaxAngularAcceleration=10000,ExternalLinearDragV=(X=0,Y=0,Z=0),ExternalLinearVelocity=(X=0,Y=0,Z=0),ExternalAngularVelocity=(X=0,Y=0,Z=0))
   ) {
    bClampLinearTranslationLimitToRefPose=true
    bOverrideWorldGravity=true
    overridePhysicsAsset=&/M_MED_StrawShave_Physics_Tie
    simulationSpace=BaseBoneSpace
    baseBoneRef {boneName=spine_05}
    componentAppliedLinearAccClamp=<0 0 0>
   }
   n9 = Component To Local(ComponentPose=@n8)
   n10 = Get bRigidBodyEnabled()
   n11 = "Transform (Modify) Bone - Bone: dyn_jacket_fr_l_01"!(
    ComponentPose=Local To Component(LocalPose=Blend Poses by bool(BlendPose_0=@n9, BlendPose_1=@n6, bActiveValue=@n10) {blendType=Linear})
    bAlphaBoolEnabled:bool=True
    Translation="5.280,-0.000,-4.390"
    Rotation="26.040,-0.000,-0.000"
    Scale="0.7,1,1"
   ) -> (Pose) {
    alphaInputType=Bool
    rotationMode=BMM_Additive
    rotationSpace=BCS_BoneSpace
    scaleMode=BMM_Additive
    scaleSpace=BCS_BoneSpace
    translationMode=BMM_Additive
    translationSpace=BCS_BoneSpace
    alphaBoolBlend {blendInTime=0.2, blendOutTime=0.2}
    boneToModify {boneName=dyn_jacket_fr_l_01}
    rotation {pitch=26.04}
    scale {x=0.7}
    translation {x=5.28, z=-4.39}
   }
   n12 = "Transform (Modify) Bone - Bone: dyn_jacket_l_01"!(ComponentPose=@n11, bAlphaBoolEnabled:bool=True, Translation="0,0,0", Rotation="13.216,7.835,1.799", Scale="1,1,1") -> (Pose) {
    alphaInputType=Bool
    rotationMode=BMM_Additive
    rotationSpace=BCS_BoneSpace
    scaleMode=BMM_Additive
    scaleSpace=BCS_BoneSpace
    translationMode=BMM_Additive
    translationSpace=BCS_BoneSpace
    alphaBoolBlend {blendInTime=0.2, blendOutTime=0.2}
    boneToModify {boneName=dyn_jacket_l_01}
    rotation=<rot 13.216 7.835 1.799>
   }
   n13 = "Transform (Modify) Bone - Bone: dyn_jacket_fr_r_01"!(ComponentPose=@n12, bAlphaBoolEnabled:bool=True, Translation="4.193,0.785,-0.706", Rotation="14.509,0.000,0.000", Scale="0.75,1,1") -> (Pose) {
    alphaInputType=Bool
    rotationMode=BMM_Additive
    rotationSpace=BCS_BoneSpace
    scaleMode=BMM_Additive
    scaleSpace=BCS_BoneSpace
    translationMode=BMM_Additive
    translationSpace=BCS_BoneSpace
    alphaBoolBlend {blendInTime=0.2, blendOutTime=0.2}
    boneToModify {boneName=dyn_jacket_fr_r_01}
    rotation {pitch=14.509}
    scale {x=0.75}
    translation=<4.193 0.785 -0.706>
   }
   n14 = "Transform (Modify) Bone - Bone: dyn_jacket_r_01"!(ComponentPose=@n13, bAlphaBoolEnabled:bool=True, Translation="0,0,0", Rotation="10.844,-19.030,-3.713", Scale="1,1,1") -> (Pose) {
    alphaInputType=Bool
    rotationMode=BMM_Additive
    rotationSpace=BCS_BoneSpace
    scaleMode=BMM_Additive
    scaleSpace=BCS_BoneSpace
    translationMode=BMM_Additive
    translationSpace=BCS_BoneSpace
    alphaBoolBlend {blendInTime=0.2, blendOutTime=0.2}
    boneToModify {boneName=dyn_jacket_r_01}
    rotation=<rot 10.844 -19.03 -3.713>
   }
   n15 = "Transform (Modify) Bone - Bone: dyn_jacket_bk_r_01"!(ComponentPose=@n14, bAlphaBoolEnabled:bool=True, Translation="0,0,0", Rotation="0,0,0", Scale="1,1,1") -> (Pose) {
    alphaInputType=Bool
    rotationMode=BMM_Additive
    rotationSpace=BCS_BoneSpace
    scaleMode=BMM_Additive
    scaleSpace=BCS_BoneSpace
    translationMode=BMM_Additive
    translationSpace=BCS_BoneSpace
    alphaBoolBlend {blendInTime=0.2, blendOutTime=0.2}
    boneToModify {boneName=dyn_jacket_bk_r_01}
   }
   n16 = "Transform (Modify) Bone - Bone: dyn_jacket_fr_l_MSTR"!(ComponentPose=@n15, bAlphaBoolEnabled:bool=True, Translation="0,0,1", Rotation="0,0,0", Scale="1,1,1") -> (Pose) {
    alphaInputType=Bool
    rotationMode=BMM_Additive
    rotationSpace=BCS_BoneSpace
    scaleMode=BMM_Additive
    scaleSpace=BCS_BoneSpace
    translationMode=BMM_Additive
    translationSpace=BCS_BoneSpace
    alphaBoolBlend {blendInTime=0.2, blendOutTime=0.2}
    boneToModify {boneName=dyn_jacket_fr_l_MSTR}
    translation {z=1}
   }
   n17 = "Transform (Modify) Bone - Bone: dyn_jacket_r_MSTR"!(ComponentPose=@n16, bAlphaBoolEnabled:bool=True, Translation="0,0,1", Rotation="0,0,0", Scale="1,1,1") -> (Pose) {
    alphaInputType=Bool
    rotationMode=BMM_Additive
    rotationSpace=BCS_BoneSpace
    scaleMode=BMM_Additive
    scaleSpace=BCS_BoneSpace
    translationMode=BMM_Additive
    translationSpace=BCS_BoneSpace
    alphaBoolBlend {blendInTime=0.2, blendOutTime=0.2}
    boneToModify {boneName=dyn_jacket_r_MSTR}
    translation {z=1}
   }
   n18 = "Transform (Modify) Bone - Bone: dyn_jacket_bk_l_MSTR"!(ComponentPose=@n17, bAlphaBoolEnabled:bool=True, Translation="0,0,0", Rotation="20,0,0", Scale="1,1,1") -> (Pose) {
    alphaInputType=Bool
    rotationMode=BMM_Additive
    rotationSpace=BCS_BoneSpace
    scaleMode=BMM_Additive
    scaleSpace=BCS_BoneSpace
    translationMode=BMM_Additive
    translationSpace=BCS_BoneSpace
    alphaBoolBlend {blendInTime=0.2, blendOutTime=0.2}
    boneToModify {boneName=dyn_jacket_bk_l_MSTR}
    rotation {pitch=20}
   }
   n19 = "Transform (Modify) Bone - Bone: dyn_jacket_bk_MSTR"!(ComponentPose=@n18, bAlphaBoolEnabled:bool=True, Translation="0,0,5", Rotation="40,0,0", Scale="1,1,1") -> (Pose) {
    alphaInputType=Bool
    rotationMode=BMM_Additive
    rotationSpace=BCS_BoneSpace
    scaleMode=BMM_Additive
    scaleSpace=BCS_BoneSpace
    translationMode=BMM_Additive
    translationSpace=BCS_BoneSpace
    alphaBoolBlend {blendInTime=0.2, blendOutTime=0.2}
    boneToModify {boneName=dyn_jacket_bk_MSTR}
    rotation {pitch=40}
    translation {z=5}
   }
   n20 = "Transform (Modify) Bone - Bone: dyn_jacket_bk_r_MSTR"!(ComponentPose=@n19, bAlphaBoolEnabled:bool=True, Translation="0,0,10", Rotation="40,0,0", Scale="1,1,1") -> (Pose) {
    alphaInputType=Bool
    rotationMode=BMM_Additive
    rotationSpace=BCS_BoneSpace
    scaleMode=BMM_Additive
    scaleSpace=BCS_BoneSpace
    translationMode=BMM_Additive
    translationSpace=BCS_BoneSpace
    alphaBoolBlend {blendInTime=0.2, blendOutTime=0.2}
    boneToModify {boneName=dyn_jacket_bk_r_MSTR}
    rotation {pitch=40}
    translation {z=10}
   }
   Output Pose(Result=Component To Local(ComponentPose=@n20)) {layerGroup=None}
   n21 = Control Rig!(Source=Copy Pose From Mesh() {bUseAttachedParent=true}, Alpha, is crouching:bool=Get bIsCrouching(), isEmotes:bool=Get bIsPlayingEmote()) -> (Pose) {controlRigAssetReference {blueprintRigClass=&/M_MED_StrawShave_CtrlRig.M_MED_StrawShave_CtrlRig_C}}
   Save cached pose 'pre_sim_cache'(Pose=@n21) {cacheName=pre_sim_cache}
  ]}
 EventGraph {nodes [
   Event BlueprintUpdateAnimation()
   > n1 = "Gravity Override State Function (Nativized)"!(
    execute
    CharacterPartAnimInstance:Custom Character Part Anim Instance Object Reference=Self-Reference()
    GravityOverrideParameterSet:Fort Gravity Override Parameter Set Structure=Get Gravity Override Params() -> (Gravity Override Params:Fort Gravity Override Parameter Set Structure)
    RigidBodyAnimNodeParameterSet:Fort Rigid Body Anim Node Parameter Set Structure=Get RBAN Params() -> (RBAN Params:Fort Rigid Body Anim Node Parameter Set Structure)
    DebugDraw:bool=false
    GameTimeInSeconds:float=GetGameTimeInSeconds()
    DebugJointName:Name=root
    StateLogicLayerName:Name=none
   ) -> (
    then
    OutGravityOverride:Vector
    OutCurrentState:Name=None
    OutGravityOverrideParameters:Fort Gravity Override Parameters Structure
    OutRigidBodyAnimNodeParameters:Fort Rigid Body Anim Node Parameters Structure
   )
   > Set Gravity Override(Gravity Override=@n1)
   > Set RBAN Settings Out(RBAN Settings Out=@n1.OutRigidBodyAnimNodeParameters)
   Event OnBeginCrouching()
   Event OnEndCrouching()
  ]}
]
variables [
 dynamics_disabled {type=boolean}
 Gravity Override {type=Vector}
 Gravity Override Params {
  type=FortGravityOverrideParameterSet
  default=(Skydive_GoingUp=(GravityOverride=(WindFrequency=(X=1.500000,Y=7.000000,Z=8.000000),WindAmplitude=(X=200.000000,Y=-500.000000,Z=400.000000),GravityOverride=(X=0.000000,Y=-300.000000,Z=100.000000),JointName="root")),Skydive_FloatingDown=(GravityOverride=(WindFrequency=(X=1.500000,Y=7.000000,Z=8.000000),WindAmplitude=(X=200.000000,Y=-500.000000,Z=400.000000),WindOffset=(X=0.000000,Y=-600.000000,Z=400.000000),GravityOverride=(X=0.000000,Y=-700.000000,Z=600.000000),JointName="root")),Skydive_DivingDown=(GravityOverride=(WindFrequency=(X=1.500000,Y=12.000000,Z=10.000000),WindAmplitude=(X=50.000000,Y=300.000000,Z=100.000000),GravityOverride=(X=-1500.000000,Y=0.000000,Z=0.000000),JointName="pelvis")),Skydive_Parachuting=(GravityOverride=(WindFrequency=(X=1.500000,Y=7.000000,Z=8.000000),WindAmplitude=(X=200.000000,Y=-500.000000,Z=400.000000),GravityOverride=(X=0.000000,Y=-800.000000,Z=-980.000000),JointName="root")),OnGround_Standing=(GravityOverride=(GravityOverride=(X=0.000000,Y=0.000000,Z=-980.000000),JointName="root")),OnGround_Moving=(GravityOverride=(GravityOverride=(X=0.000000,Y=0.000000,Z=-980.000000),JointName="root")),OnGround_Moving_HighRotationRate=(GravityOverride=(GravityOverride=(X=0.000000,Y=0.000000,Z=-980.000000),JointName="root")),OnGround_Crouch_Moving=(GravityOverride=(GravityOverride=(X=0.000000,Y=0.000000,Z=-980.000000),JointName="root")),Emote_or_Melee=(GravityOverride=(GravityOverride=(X=0.000000,Y=0.000000,Z=-980.000000),JointName="root")),Swimming=(GravityOverride=(GravityOverride=(X=-800.000000,Y=0.000000,Z=0.000000),JointName="pelvis")),Front_End=(GravityOverride=(GravityOverride=(X=0.000000,Y=0.000000,Z=-980.000000),JointName="root")),DBNO=(GravityOverride=(GravityOverride=(X=-980.000000,Y=400.000000,Z=0.000000),JointName="pelvis")),DBNO_Carried=(GravityOverride=(GravityOverride=(X=-980.000000,Y=500.000000,Z=-500.000000),JointName="pelvis")),OnGround_Falling=(GravityOverride=(GravityOverride=(X=0.000000,Y=0.000000,Z=-980.000000),JointName="root")),Ziplining=(GravityOverride=(GravityOverride=(X=0.000000,Y=0.000000,Z=-980.000000),JointName="root")),Skydive_Surfing=(GravityOverride=(WindFrequency=(X=1.500000,Y=7.000000,Z=8.000000),WindAmplitude=(X=200.000000,Y=-500.000000,Z=400.000000),GravityOverride=(X=500.000000,Y=-300.000000,Z=-980.000000),JointName="root")),Riding_Standing=(GravityOverride=(GravityOverride=(X=0.000000,Y=-350.000000,Z=-980.000000),JointName="root")),Riding_Moving=(GravityOverride=(GravityOverride=(X=0.000000,Y=-450.000000,Z=-450.000000),JointName="root")),Riding_Moving_Targeting=(GravityOverride=(GravityOverride=(X=0.000000,Y=-200.000000,Z=-980.000000),JointName="root")),Tactical_Sprint=(GravityOverride=(GravityOverride=(X=0.000000,Y=-450.000000,Z=-500.000000),JointName="root")),Rail_Grinding=(GravityOverride=(GravityOverride=(X=0.000000,Y=0.000000,Z=-980.000000),JointName="root")),)
 }
 RBAN Params {
  type=FortRigidBodyAnimNodeParameterSet
  default=(Skydive_GoingUp=(RigidBodyAnimNode=(ComponentLinearAccScale=(X=0.000000,Y=-0.500000,Z=0.000000),ComponentLinearVelScale=(X=0.000000,Y=-0.500000,Z=0.000000),SimSpaceSettings=(WorldAlpha=0.500000,VelocityScaleZ=0.500000,DampingAlpha=0.000000,MaxLinearVelocity=250.000000,MaxAngularVelocity=6.000000,MaxLinearAcceleration=500.000000,MaxAngularAcceleration=24.000000))),Skydive_FloatingDown=(RigidBodyAnimNode=(ComponentLinearAccScale=(X=0.000000,Y=0.500000,Z=0.000000),ComponentLinearVelScale=(X=0.000000,Y=0.500000,Z=0.000000),SimSpaceSettings=(WorldAlpha=0.500000,VelocityScaleZ=0.500000,DampingAlpha=0.000000,MaxLinearVelocity=250.000000,MaxAngularVelocity=6.000000,MaxLinearAcceleration=500.000000,MaxAngularAcceleration=24.000000))),Skydive_DivingDown=(RigidBodyAnimNode=(ComponentLinearAccScale=(X=0.100000,Y=0.000000,Z=0.000000),ComponentLinearVelScale=(X=0.100000,Y=0.000000,Z=0.000000),SimSpaceSettings=(WorldAlpha=0.500000,VelocityScaleZ=0.500000,DampingAlpha=0.000000,MaxLinearVelocity=250.000000,MaxAngularVelocity=6.000000,MaxLinearAcceleration=500.000000,MaxAngularAcceleration=24.000000))),Skydive_Parachuting=(RigidBodyAnimNode=(ComponentLinearAccScale=(X=0.000000,Y=0.100000,Z=0.300000),ComponentLinearVelScale=(X=0.000000,Y=0.100000,Z=0.300000),SimSpaceSettings=(WorldAlpha=0.500000,VelocityScaleZ=0.500000,DampingAlpha=0.000000,MaxLinearVelocity=250.000000,MaxAngularVelocity=6.000000,MaxLinearAcceleration=500.000000,MaxAngularAcceleration=24.000000))),OnGround_Standing=(RigidBodyAnimNode=(ComponentLinearAccScale=(X=0.300000,Y=0.000000,Z=0.000000),ComponentLinearVelScale=(X=0.300000,Y=0.000000,Z=0.000000),SimSpaceSettings=(WorldAlpha=0.500000,VelocityScaleZ=0.500000,DampingAlpha=0.000000,MaxLinearVelocity=250.000000,MaxAngularVelocity=6.000000,MaxLinearAcceleration=500.000000,MaxAngularAcceleration=24.000000))),OnGround_Moving=(RigidBodyAnimNode=(ComponentLinearAccScale=(X=0.300000,Y=0.000000,Z=0.000000),ComponentLinearVelScale=(X=0.300000,Y=0.000000,Z=0.000000),SimSpaceSettings=(WorldAlpha=0.500000,VelocityScaleZ=0.500000,DampingAlpha=0.000000,MaxLinearVelocity=250.000000,MaxAngularVelocity=6.000000,MaxLinearAcceleration=500.000000,MaxAngularAcceleration=24.000000))),OnGround_Moving_HighRotationRate=(RigidBodyAnimNode=(SimSpaceSettings=(WorldAlpha=0.500000,VelocityScaleZ=0.500000,DampingAlpha=0.000000,MaxLinearVelocity=250.000000,MaxAngularVelocity=6.000000,MaxLinearAcceleration=500.000000,MaxAngularAcceleration=24.000000))),OnGround_Crouch_Moving=(RigidBodyAnimNode=(SimSpaceSettings=(WorldAlpha=0.500000,VelocityScaleZ=0.500000,DampingAlpha=0.000000,MaxLinearVelocity=250.000000,MaxAngularVelocity=6.000000,MaxLinearAcceleration=500.000000,MaxAngularAcceleration=24.000000))),Emote_or_Melee=(RigidBodyAnimNode=(SimSpaceSettings=(WorldAlpha=0.500000,VelocityScaleZ=0.500000,DampingAlpha=0.000000,MaxLinearVelocity=250.000000,MaxAngularVelocity=6.000000,MaxLinearAcceleration=500.000000,MaxAngularAcceleration=24.000000))),Swimming=(RigidBodyAnimNode=(SimSpaceSettings=(WorldAlpha=0.500000,VelocityScaleZ=0.500000,DampingAlpha=0.000000,MaxLinearVelocity=250.000000,MaxAngularVelocity=6.000000,MaxLinearAcceleration=500.000000,MaxAngularAcceleration=24.000000))),Front_End=(RigidBodyAnimNode=(SimSpaceSettings=(WorldAlpha=0.500000,VelocityScaleZ=0.500000,DampingAlpha=0.000000,MaxLinearVelocity=250.000000,MaxAngularVelocity=6.000000,MaxLinearAcceleration=500.000000,MaxAngularAcceleration=24.000000))),DBNO=(RigidBodyAnimNode=(SimSpaceSettings=(WorldAlpha=0.500000,VelocityScaleZ=0.500000,DampingAlpha=0.000000,MaxLinearVelocity=250.000000,MaxAngularVelocity=6.000000,MaxLinearAcceleration=500.000000,MaxAngularAcceleration=24.000000))),DBNO_Carried=(RigidBodyAnimNode=(SimSpaceSettings=(WorldAlpha=0.500000,VelocityScaleZ=0.500000,DampingAlpha=0.000000,MaxLinearVelocity=250.000000,MaxAngularVelocity=6.000000,MaxLinearAcceleration=500.000000,MaxAngularAcceleration=24.000000))),OnGround_Falling=(RigidBodyAnimNode=(SimSpaceSettings=(WorldAlpha=0.500000,VelocityScaleZ=0.500000,DampingAlpha=0.000000,MaxLinearVelocity=250.000000,MaxAngularVelocity=6.000000,MaxLinearAcceleration=500.000000,MaxAngularAcceleration=24.000000))),Ziplining=(RigidBodyAnimNode=(SimSpaceSettings=(WorldAlpha=0.500000,VelocityScaleZ=0.500000,DampingAlpha=0.000000,MaxLinearVelocity=250.000000,MaxAngularVelocity=6.000000,MaxLinearAcceleration=500.000000,MaxAngularAcceleration=24.000000))),Skydive_Surfing=(RigidBodyAnimNode=(SimSpaceSettings=(WorldAlpha=0.500000,VelocityScaleZ=0.500000,DampingAlpha=0.000000,MaxLinearVelocity=250.000000,MaxAngularVelocity=6.000000,MaxLinearAcceleration=500.000000,MaxAngularAcceleration=24.000000))),Riding_Standing=(RigidBodyAnimNode=(SimSpaceSettings=(WorldAlpha=0.500000,VelocityScaleZ=0.500000,DampingAlpha=0.000000,MaxLinearVelocity=250.000000,MaxAngularVelocity=6.000000,MaxLinearAcceleration=500.000000,MaxAngularAcceleration=24.000000))),Riding_Moving=(RigidBodyAnimNode=(SimSpaceSettings=(WorldAlpha=0.500000,VelocityScaleZ=0.500000,DampingAlpha=0.000000,MaxLinearVelocity=250.000000,MaxAngularVelocity=6.000000,MaxLinearAcceleration=500.000000,MaxAngularAcceleration=24.000000))),Riding_Moving_Targeting=(RigidBodyAnimNode=(SimSpaceSettings=(WorldAlpha=0.500000,VelocityScaleZ=0.500000,DampingAlpha=0.000000,MaxLinearVelocity=250.000000,MaxAngularVelocity=6.000000,MaxLinearAcceleration=500.000000,MaxAngularAcceleration=24.000000))),Tactical_Sprint=(RigidBodyAnimNode=(ComponentLinearAccScale=(X=0.300000,Y=0.000000,Z=0.000000),ComponentLinearVelScale=(X=0.300000,Y=0.000000,Z=0.000000),SimSpaceSettings=(WorldAlpha=0.500000,VelocityScaleZ=0.500000,DampingAlpha=0.000000,MaxLinearVelocity=250.000000,MaxAngularVelocity=6.000000,MaxLinearAcceleration=500.000000,MaxAngularAcceleration=24.000000))),Rail_Grinding=(RigidBodyAnimNode=(SimSpaceSettings=(WorldAlpha=0.500000,VelocityScaleZ=0.500000,DampingAlpha=0.000000,MaxLinearVelocity=250.000000,MaxAngularVelocity=6.000000,MaxLinearAcceleration=500.000000,MaxAngularAcceleration=24.000000))),)
 }
 RBAN Settings Out {
  type=FortRigidBodyAnimNodeParameters
  default {simSpaceSettings {
    dampingAlpha=0
    maxAngularAcceleration=16
    maxAngularVelocity=4
    maxLinearAcceleration=500
    maxLinearVelocity=200
    velocityScaleZ=0.5
    worldAlpha=0.5
   }}
 }
]
