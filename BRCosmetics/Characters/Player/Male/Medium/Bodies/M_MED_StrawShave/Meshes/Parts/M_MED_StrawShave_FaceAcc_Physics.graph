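# Physics asset for the face accessory mesh of the StrawShave male medium character, providing capsule bodies and constraints for its hair bones.
# A large head capsule anchors fourteen small base-hair capsules through joints with free swing and twist.
# Each base leads a chain of one to three larger, offset dynamic capsules.
# Chain joints share a very tight twist limit, while swing limits vary by strand from stiff to looser.
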PhysicsAsset /BRCosmetics/Characters/Player/Male/Medium/Bodies/M_MED_StrawShave/Meshes/Parts/M_MED_StrawShave_FaceAcc_Physics
bodies {
 base_hair1 [{shapeType=Capsule, length=0.505, radius=0.505, shapeName=base_hair1_capsule, rotation=<rot 0 -90 0>}]
 base_hair10 [{shapeType=Capsule, length=0.505, radius=0.505, shapeName=base_hair10_capsule, rotation=<rot 0 -90 0>}]
 base_hair11 [{shapeType=Capsule, length=0.505, radius=0.505, shapeName=base_hair11_capsule, rotation=<rot 0 -90 0>}]
 base_hair12 [{shapeType=Capsule, length=0.505, radius=0.505, shapeName=base_hair12_capsule, rotation=<rot 0 -90 0>}]
 base_hair13 [{shapeType=Capsule, length=0.505, radius=0.505, shapeName=base_hair13_capsule, rotation=<rot 0 -90 0>}]
 base_hair14 [{shapeType=Capsule, length=0.505, radius=0.505, shapeName=base_hair14_capsule, rotation=<rot 0 -90 0>}]
 base_hair2 [{shapeType=Capsule, length=0.505, radius=0.505, shapeName=base_hair2_capsule, rotation=<rot 0 -90 0>}]
 base_hair3 [{shapeType=Capsule, length=0.505, radius=0.505, shapeName=base_hair3_capsule, rotation=<rot 0 -90 0>}]
 base_hair4 [{shapeType=Capsule, length=0.505, radius=0.505, shapeName=base_hair4_capsule, rotation=<rot 0 -90 0>}]
 base_hair5 [{shapeType=Capsule, length=0.505, radius=0.505, shapeName=base_hair5_capsule, rotation=<rot 0 -90 0>}]
 base_hair6 [{shapeType=Capsule, length=0.505, radius=0.505, shapeName=base_hair6_capsule, rotation=<rot 0 -90 0>}]
 base_hair7 [{shapeType=Capsule, length=0.505, radius=0.505, shapeName=base_hair7_capsule, rotation=<rot 0 -90 0>}]
 base_hair8 [{shapeType=Capsule, length=0.505, radius=0.505, shapeName=base_hair8_capsule, rotation=<rot 0 -90 0>}]
 base_hair9 [{shapeType=Capsule, length=0.505, radius=0.505, shapeName=base_hair9_capsule, rotation=<rot 0 -90 0>}]
 dyn_hair10_1 [{shapeType=Capsule, length=5, radius=4, shapeName=dyn_hair10_1_capsule, center=<2 0 3.5021>, rotation=<rot 90 -90 0>}]
 dyn_hair10_2 [{shapeType=Capsule, length=5, radius=4, shapeName=dyn_hair10_2_capsule, center=<2 0 3.5021>, rotation=<rot 90 -90 0>}]
 dyn_hair11_1 [{shapeType=Capsule, length=5, radius=4, shapeName=dyn_hair11_1_capsule, center=<2 0 3.5021>, rotation=<rot 90 -90 0>}]
 dyn_hair11_2 [{shapeType=Capsule, length=5, radius=4, shapeName=dyn_hair11_2_capsule, center=<2 0 3.5021>, rotation=<rot 90 -90 0>}]
 dyn_hair12_1 [{shapeType=Capsule, length=5, radius=4, shapeName=dyn_hair12_1_capsule, center=<2 0 3.5021>, rotation=<rot 90 -90 0>}]
 dyn_hair12_2 [{shapeType=Capsule, length=5, radius=4, shapeName=dyn_hair12_2_capsule, center=<2 0 3.5021>, rotation=<rot 90 -90 0>}]
 dyn_hair13_1 [{shapeType=Capsule, length=5, radius=4, shapeName=dyn_hair13_1_capsule, center=<2 0.9468 0.9601>, rotation=<rot 45.4014 90 180>}]
 dyn_hair13_2 [{shapeType=Capsule, length=5, radius=4, shapeName=dyn_hair13_2_capsule, center=<2 0.9468 0.9601>, rotation=<rot 45.4014 90 180>}]
 dyn_hair14_1 [{shapeType=Capsule, length=5, radius=6, shapeName=dyn_hair14_1_capsule, center=<2 0 -4.8422>, rotation=<rot 90 -90 0>}]
 dyn_hair1_1 [{shapeType=Capsule, length=5, radius=4, shapeName=dyn_hair1_1_capsule, center=<2 1.7685 2.3598>, rotation=<rot 61.8105 90 -180>}]
 dyn_hair1_2 [{shapeType=Capsule, length=5, radius=4, shapeName=dyn_hair1_2_capsule, center=<2 1.6383 4.4771>, rotation=<rot 63.8136 90 -180>}]
 dyn_hair1_3 [{shapeType=Capsule, length=5, radius=4, shapeName=dyn_hair1_3_capsule, center=<4.7566 1.555 4.331>, rotation=<rot 63.8136 90.0001 169.0097>}]
 dyn_hair2_1 [{shapeType=Capsule, length=5, radius=4, shapeName=dyn_hair2_1_capsule, center=<2 -4.6712 1.5206>, rotation=<rot -25.9782 90 -180>}]
 dyn_hair2_2 [{shapeType=Capsule, length=5, radius=4, shapeName=dyn_hair2_2_capsule, center=<4.8023 -4.4732 0.5969>, rotation=<rot -15.2581 90 -180>}]
 dyn_hair3_1 [{shapeType=Capsule, length=5, radius=4, shapeName=dyn_hair3_1_capsule, center=<2 -1.6383 0>, rotation=<rot 0 90 -180>}]
 dyn_hair3_2 [{shapeType=Capsule, length=5, radius=4, shapeName=dyn_hair3_2_capsule, center=<2 -1.6383 0>, rotation=<rot 0 90 -180>}]
 dyn_hair4_1 [{shapeType=Capsule, length=5, radius=4, shapeName=dyn_hair4_1_capsule, center=<2 -3.0217 0.3686>, rotation=<rot 6.9541 -90 0>}]
 dyn_hair4_2 [{shapeType=Capsule, length=5, radius=4, shapeName=dyn_hair4_2_capsule, center=<2 -3.0217 0.3686>, rotation=<rot 6.9541 -90 0>}]
 dyn_hair5_1 [{shapeType=Capsule, length=5, radius=4, shapeName=dyn_hair5_1_capsule, center=<2 -3.0217 0.3686>, rotation=<rot 6.9541 -90 0>}]
 dyn_hair5_2 [{shapeType=Capsule, length=5, radius=4, shapeName=dyn_hair5_2_capsule, center=<2 -3.0217 0.3686>, rotation=<rot 6.9541 -90 0>}]
 dyn_hair6_1 [{shapeType=Capsule, length=5, radius=4, shapeName=dyn_hair6_1_capsule, center=<2 -3.0217 0.3686>, rotation=<rot 6.9541 -90 0>}]
 dyn_hair6_2 [{shapeType=Capsule, length=5, radius=4, shapeName=dyn_hair6_2_capsule, center=<2 -3.0217 0.3686>, rotation=<rot 6.9541 -90 0>}]
 dyn_hair7_1 [{
   shapeType=Capsule
   length=5
   radius=4
   shapeName=dyn_hair7_1_capsule
   center=<1.9997 -1.3074 4.0911>
   rotation=<rot 58.8781 -90.0085 -0.0012>
  }]
 dyn_hair7_2 [{shapeType=Capsule, length=5, radius=4, shapeName=dyn_hair7_2_capsule, center=<3.1252 -1.5165 4.1015>, rotation=<rot 65.4985 -90 0>}]
 dyn_hair7_3 [{shapeType=Capsule, length=5, radius=4, shapeName=dyn_hair7_3_capsule, center=<6.1683 -2.2759 3.577>, rotation=<rot 62.8839 -90.0001 -3.0963>}]
 dyn_hair8_1 [{shapeType=Capsule, length=5, radius=4, shapeName=dyn_hair8_1_capsule, center=<2 -1.1594 2.252>, rotation=<rot 62.7604 -90 0>}]
 dyn_hair8_2 [{shapeType=Capsule, length=5, radius=4, shapeName=dyn_hair8_2_capsule, center=<2 -0.9611 1.867>, rotation=<rot 62.7604 -90 0>}]
 dyn_hair9_1 [{shapeType=Capsule, length=5, radius=4, shapeName=dyn_hair9_1_capsule, center=<2 -0.6175 3.7585>, rotation=<rot 80.6704 -90 0>}]
 dyn_hair9_2 [{shapeType=Capsule, length=5, radius=4, shapeName=dyn_hair9_2_capsule, center=<2 -0.4023 2.4486>, rotation=<rot 80.6704 -90 0>}]
 head [{shapeType=Capsule, length=15, radius=13.5, shapeName=head_capsule, center=<11.791 2.281 0>, rotation=<rot 90 0 0>}]
}
constraints [
 {bone1Name=dyn_hair1_1, bone2Name=base_hair1, swing1LimitDegrees=5, swing2LimitDegrees=5, twistLimitDegrees=1}
 {bone1Name=dyn_hair1_2, bone2Name=dyn_hair1_1, swing1LimitDegrees=5, swing2LimitDegrees=5, twistLimitDegrees=1}
 {bone1Name=dyn_hair1_3, bone2Name=dyn_hair1_2, swing1LimitDegrees=5, swing2LimitDegrees=5, twistLimitDegrees=1}
 {bone1Name=dyn_hair2_1, bone2Name=base_hair2, swing1LimitDegrees=10, swing2LimitDegrees=10, twistLimitDegrees=1}
 {bone1Name=dyn_hair2_2, bone2Name=dyn_hair2_1, swing1LimitDegrees=10, swing2LimitDegrees=10, twistLimitDegrees=1}
 {bone1Name=dyn_hair3_1, bone2Name=base_hair3, swing1LimitDegrees=10, swing2LimitDegrees=10, twistLimitDegrees=1}
 {bone1Name=dyn_hair3_2, bone2Name=dyn_hair3_1, swing1LimitDegrees=10, swing2LimitDegrees=10, twistLimitDegrees=1}
 {bone1Name=dyn_hair4_1, bone2Name=base_hair4, swing1LimitDegrees=25, swing2LimitDegrees=25, twistLimitDegrees=1}
 {bone1Name=dyn_hair4_2, bone2Name=dyn_hair4_1, swing1LimitDegrees=25, swing2LimitDegrees=25, twistLimitDegrees=1}
 {bone1Name=dyn_hair5_1, bone2Name=base_hair5, swing1LimitDegrees=25, swing2LimitDegrees=25, twistLimitDegrees=1}
 {bone1Name=dyn_hair5_2, bone2Name=dyn_hair5_1, swing1LimitDegrees=25, swing2LimitDegrees=25, twistLimitDegrees=1}
 {bone1Name=dyn_hair6_1, bone2Name=base_hair6, swing1LimitDegrees=25, swing2LimitDegrees=25, twistLimitDegrees=1}
 {bone1Name=dyn_hair6_2, bone2Name=dyn_hair6_1, swing1LimitDegrees=25, swing2LimitDegrees=25, twistLimitDegrees=1}
 {bone1Name=dyn_hair7_1, bone2Name=base_hair7, swing1LimitDegrees=5, swing2LimitDegrees=5, twistLimitDegrees=1}
 {bone1Name=dyn_hair7_2, bone2Name=dyn_hair7_1, swing1LimitDegrees=5, swing2LimitDegrees=5, twistLimitDegrees=1}
 {bone1Name=dyn_hair7_3, bone2Name=dyn_hair7_2, swing1LimitDegrees=5, swing2LimitDegrees=5, twistLimitDegrees=1}
 {bone1Name=dyn_hair8_1, bone2Name=base_hair8, swing1LimitDegrees=25, swing2LimitDegrees=25, twistLimitDegrees=1}
 {bone1Name=dyn_hair8_2, bone2Name=dyn_hair8_1, swing1LimitDegrees=25, swing2LimitDegrees=25, twistLimitDegrees=1}
 {bone1Name=dyn_hair9_1, bone2Name=base_hair9, swing1LimitDegrees=25, swing2LimitDegrees=25, twistLimitDegrees=1}
 {bone1Name=dyn_hair9_2, bone2Name=dyn_hair9_1, swing1LimitDegrees=25, swing2LimitDegrees=25, twistLimitDegrees=1}
 {bone1Name=dyn_hair10_1, bone2Name=base_hair10, swing1LimitDegrees=25, swing2LimitDegrees=25, twistLimitDegrees=1}
 {bone1Name=dyn_hair10_2, bone2Name=dyn_hair10_1, swing1LimitDegrees=25, swing2LimitDegrees=25, twistLimitDegrees=1}
 {bone1Name=dyn_hair11_1, bone2Name=base_hair11, swing1LimitDegrees=25, swing2LimitDegrees=25, twistLimitDegrees=1}
 {bone1Name=dyn_hair11_2, bone2Name=dyn_hair11_1, swing1LimitDegrees=25, swing2LimitDegrees=25, twistLimitDegrees=1}
 {bone1Name=dyn_hair12_1, bone2Name=base_hair12, swing1LimitDegrees=25, swing2LimitDegrees=25, twistLimitDegrees=1}
 {bone1Name=dyn_hair12_2, bone2Name=dyn_hair12_1, swing1LimitDegrees=25, swing2LimitDegrees=25, twistLimitDegrees=1}
 {bone1Name=dyn_hair13_1, bone2Name=base_hair13, swing1LimitDegrees=10, swing2LimitDegrees=10, twistLimitDegrees=1}
 {bone1Name=dyn_hair13_2, bone2Name=dyn_hair13_1, swing1LimitDegrees=10, swing2LimitDegrees=10, twistLimitDegrees=1}
 {bone1Name=dyn_hair14_1, bone2Name=base_hair14, swing1LimitDegrees=10, swing2LimitDegrees=10, twistLimitDegrees=1}
 {
  bone1Name=base_hair1
  bone2Name=head
  swing1LimitDegrees=45
  swing1Motion=Free
  swing2LimitDegrees=45
  swing2Motion=Free
  twistLimitDegrees=45
  twistMotion=Free
 }
 {
  bone1Name=base_hair2
  bone2Name=head
  swing1LimitDegrees=45
  swing1Motion=Free
  swing2LimitDegrees=45
  swing2Motion=Free
  twistLimitDegrees=45
  twistMotion=Free
 }
 {
  bone1Name=base_hair3
  bone2Name=head
  swing1LimitDegrees=45
  swing1Motion=Free
  swing2LimitDegrees=45
  swing2Motion=Free
  twistLimitDegrees=45
  twistMotion=Free
 }
 {
  bone1Name=base_hair4
  bone2Name=head
  swing1LimitDegrees=45
  swing1Motion=Free
  swing2LimitDegrees=45
  swing2Motion=Free
  twistLimitDegrees=45
  twistMotion=Free
 }
 {
  bone1Name=base_hair5
  bone2Name=head
  swing1LimitDegrees=45
  swing1Motion=Free
  swing2LimitDegrees=45
  swing2Motion=Free
  twistLimitDegrees=45
  twistMotion=Free
 }
 {
  bone1Name=base_hair6
  bone2Name=head
  swing1LimitDegrees=45
  swing1Motion=Free
  swing2LimitDegrees=45
  swing2Motion=Free
  twistLimitDegrees=45
  twistMotion=Free
 }
 {
  bone1Name=base_hair7
  bone2Name=head
  swing1LimitDegrees=45
  swing1Motion=Free
  swing2LimitDegrees=45
  swing2Motion=Free
  twistLimitDegrees=45
  twistMotion=Free
 }
 {
  bone1Name=base_hair8
  bone2Name=head
  swing1LimitDegrees=45
  swing1Motion=Free
  swing2LimitDegrees=45
  swing2Motion=Free
  twistLimitDegrees=45
  twistMotion=Free
 }
 {
  bone1Name=base_hair9
  bone2Name=head
  swing1LimitDegrees=45
  swing1Motion=Free
  swing2LimitDegrees=45
  swing2Motion=Free
  twistLimitDegrees=45
  twistMotion=Free
 }
 {
  bone1Name=base_hair10
  bone2Name=head
  swing1LimitDegrees=45
  swing1Motion=Free
  swing2LimitDegrees=45
  swing2Motion=Free
  twistLimitDegrees=45
  twistMotion=Free
 }
 {
  bone1Name=base_hair11
  bone2Name=head
  swing1LimitDegrees=45
  swing1Motion=Free
  swing2LimitDegrees=45
  swing2Motion=Free
  twistLimitDegrees=45
  twistMotion=Free
 }
 {
  bone1Name=base_hair12
  bone2Name=head
  swing1LimitDegrees=45
  swing1Motion=Free
  swing2LimitDegrees=45
  swing2Motion=Free
  twistLimitDegrees=45
  twistMotion=Free
 }
 {
  bone1Name=base_hair13
  bone2Name=head
  swing1LimitDegrees=45
  swing1Motion=Free
  swing2LimitDegrees=45
  swing2Motion=Free
  twistLimitDegrees=45
  twistMotion=Free
 }
 {
  bone1Name=base_hair14
  bone2Name=head
  swing1LimitDegrees=45
  swing1Motion=Free
  swing2LimitDegrees=45
  swing2Motion=Free
  twistLimitDegrees=45
  twistMotion=Free
 }
]
